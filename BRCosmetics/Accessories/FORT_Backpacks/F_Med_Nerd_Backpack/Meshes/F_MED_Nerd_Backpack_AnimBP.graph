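# Animation Blueprint for the F_Med_Nerd backpack cosmetic that adds physics motion to its key and key ring bones.
# The input pose passes through two chained Anim Dynamics nodes in component space, first drn_key, then drn_key_ring, before returning to local space.
# The key only slides slightly along X, while the ring swings within narrow Y/Z limits.
# Both use angular springs, heavy damping and a simulation-space gravity override, with no event graph logic.
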AnimBlueprint /BRCosmetics/Accessories/FORT_Backpacks/F_Med_Nerd_Backpack/Meshes/F_MED_Nerd_Backpack_AnimBP
target_skeleton=&/F_MED_Nerd_Backpack_Skeleton
graphs [
 AnimGraph {nodes [
   AnimGraphNode_SubInput_0 = Input Pose()
   n1 = Anim Dynamics - Bone: drn_key(ComponentPose=Local To Component(LocalPose=@AnimGraphNode_SubInput_0), Alpha=1, ExternalForce="0,0,0") {
    angularDampingOverride=0.999
    angularSpringConstant=1000
    bAngularSpring=true
    bGravityOverrideInSimSpace=true
    bOverrideAngularDamping=true
    bOverrideLinearDamping=true
    bUseGravityOverride=true
    linearDampingOverride=0.999
    alphaBoolBlend {blendOption=Linear}
    boundBone {boneName=drn_key}
    componentAppliedLinearAccClamp=<1000 1000 1000>
    componentLinearAccScale=<1 1 1>
    componentLinearVelScale=<1 1 1>
    gravityOverride {z=-980}
    physicsBodyDefinitions [{
      sphereCollisionRadius=0
      boundBone {boneName=drn_key}
      boxExtents=<2 0.5 0.5>
      constraintSetup {
       angularTargetAxis=AxisY
       angularLimitsMax=<0 0 0>
       angularLimitsMin=<0 0 0>
       angularTarget=<0 0 0>
       linearAxesMax=<0.02 0 0>
       linearAxesMin=<-0.08 0 0>
      }
      localJointOffset=<0.1 0 0>
     }]
   }
   n2 = Anim Dynamics - Bone: drn_key_ring(ComponentPose=@n1, Alpha=1, ExternalForce="0,0,0") {
    angularDampingOverride=0.999
    angularSpringConstant=1000
    bAngularSpring=true
    bGravityOverrideInSimSpace=true
    bOverrideAngularDamping=true
    bOverrideLinearDamping=true
    bUseGravityOverride=true
    linearDampingOverride=0.999
    alphaBoolBlend {blendOption=Linear}
    boundBone {boneName=drn_key_ring}
    componentAppliedLinearAccClamp=<1000 1000 1000>
    componentLinearAccScale=<1 1 1>
    componentLinearVelScale=<1 1 1>
    gravityOverride {z=-980}
    physicsBodyDefinitions [{
      sphereCollisionRadius=0
      boundBone {boneName=drn_key_ring}
      boxExtents=<1 5 5>
      constraintSetup {
       angularTargetAxis=AxisY
       angularLimitsMax=<0 20 15>
       angularLimitsMin=<0 -20 -15>
       angularTarget=<0 0 0>
       linearAxesMax=<0.01 0.01 0.01>
       linearAxesMin=<-0.01 -0.01 -0.01>
      }
      localJointOffset=<1 0 0>
     }]
   }
   Output Pose(Result=Component To Local(ComponentPose=@n2)) {layerGroup=None}
  ]}
 EventGraph {nodes [Event BlueprintUpdateAnimation()]}
]
variables []
